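# Masked, instancing-capable material for blending Season 9 volcano crater surfaces between lava ground, lava soil and tinted rock mud.
# A height lerp driven by vertex blue, UV1, tiling noise and lava-ground height produces the opacity mask.
# One branch mixes world-projected ground and soil with optional slope fading and a lava edge gradient.
# The mud layer nudges vertices toward the camera against z-fighting; another branch tints lava ground using the same mask.
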Material /Figment_S09/Athena/Environments/Volcano/Materials/M_Athena_VolcanoCrater_Blend
graph {
 properties {
  bUseMaterialAttributes=true
  bUsedWithInstancedStaticMeshes=true
  blendMode=BLEND_Masked
  pixelDepthOffsetMode=PDOM_Legacy
  usedShadingModels=Default Lit
 }
 nodes [
  ComponentMask(None=WorldPosition().XYZ) {g=true, r=true}
  n1 = VectorParameter() {parameterName=TIntLavaTransition, defaultValue=<rgba 1 1 1 1>}
  n2 = TextureCoordinate() {coordinateIndex=1}
  LinearInterpolate(
   A=@n1.RGB
   B=VectorParameter() {parameterName=LavaTintAO, defaultValue=<rgba 1 1 1 1>}.RGB
   Alpha=Power(Base=ComponentMask(None=@n2) {g=true}, Exp=ScalarParameter() {defaultValue=0.3696, parameterName=AOpower})
  )
  n3 = Divide(A=ComponentMask(None=WorldPosition().XYZ) {g=true, r=true}, B=Constant() {r=921.5037})
  n4 = MaterialFunctionCall!(UV=@n3, WorldAlignedSize) {materialFunction=/Game/Environments/Sets/Vulcano/Materials/MF_LavaGround}
  n5 = Constant()
  n6 = Constant() {r=1.5828}
  n7 = WorldPosition()
  n8 = TextureSample(UVs=Divide(A=ComponentMask(None=@n7.XYZ) {g=true, r=true}, B=Constant() {r=2655.5642})) {texture=/Figment_S09/Effects/Fort_Effects/Textures/Misc/T_Tiling_Noise_01}
  n9 = MaterialFunctionCall(
   Transition Phase=Clamp(None=LinearInterpolate(A=@n5, B=@n6, Alpha=@n8.R))
   Height Texture=OneMinus(None=@n4.Height)
   Contrast=Constant() {r=0.0476}
  ) {materialFunction=&HeightLerp}
  n10 = VertexNormalWS()
  n11 = Clamp(None=LinearInterpolate(A=Constant() {r=-7.9409}, B=Constant() {r=9.2509}, Alpha=ComponentMask(None=@n10) {b=true}))
  n12 = MaterialFunctionCall!(
   Base Material=@n4.Result
   Top Material=MaterialFunctionCall!(UVs) {materialFunction=/Game/Environments/Sets/Vulcano/Materials/MF_Lava_Soil}.Result
   Alpha=StaticSwitchParameter(True=Multiply(A=@n9.Alpha, B=@n11), False=@n9.Alpha) {parameterName=FadeSandonWalls}
  ) {materialFunction=/Engine/Functions/MaterialLayerFunctions/MatLayerBlend_Standard}
  n13 = VertexColor()
  n14 = Divide(A=TextureCoordinate(), B=ScalarParameter() {defaultValue=5, parameterName=Tiling})
  n15 = TextureSample(UVs=@n14) {texture=/Game/Packages/DS_Fortnight/WhiteRockCustomBR_D}
  n16 = ComponentMask(None=TextureCoordinate() {coordinateIndex=1}) {g=true}
  n17 = Saturate(None=Add(A=VertexColor().B, B=@n16))
  n18 = ScalarParameter() {defaultValue=0.2878, parameterName=TranSitionPower}
  n19 = Divide(A=TextureCoordinate(), B=ScalarParameter() {defaultValue=6.3393, parameterName=TilingNoise})
  n20 = ScalarParameter() {defaultValue=6, parameterName=HeightMod}
  n21 = ScalarParameter() {defaultValue=-2.6014, parameterName=HeightMod2}
  n22 = MaterialFunctionCall(UV=Multiply(A=TextureCoordinate(), B=ScalarParameter() {defaultValue=0.5, parameterName=LavaSoil_Tiling})) {materialFunction=&MF_LavaGround}
  n23 = ScalarParameter() {parameterName=MudBlendContrast}
  n24 = MaterialFunctionCall!(
   Diffuse 1
   Diffuse 2
   Transistion Phase=Multiply(A=Power(Base=@n17, Exp=@n18), B=ScalarParameter() {defaultValue=1, parameterName=TranSition})
   Height Texture 1=Multiply(
    A=TextureSample(UVs=@n19) {texture=&T_Tiling_Noise_01}.R
    B=LinearInterpolate(A=@n20, B=@n21, Alpha=@n17)
   )
   Height Texture 2=OneMinus(None=@n22.Height)
   Contrast=@n23
  ) {materialFunction=/Engine/Functions/Engine_MaterialFunctions02/Texturing/HeightLerpWithTwoHeightMaps}
  n25 = Subtract(A=CameraPositionWS(), B=WorldPosition().XYZ)
  n26 = MakeMaterialAttributes!(
   BaseColor=Multiply(A=@n15.RGB, B=VectorParameter() {parameterName=MudTint, defaultValue=<rgba 0.2396 0.228 0.2084 0>}.RGB)
   Metallic
   Specular=Constant() {r=0.2}
   Roughness=Constant() {r=1}
   Anisotropy
   EmissiveColor
   Opacity
   OpacityMask=@n24.Alpha
   Normal=TextureSample(UVs=@n14) {samplerType=SAMPLERTYPE_Normal, texture=/Game/Packages/DS_Fortnight/WhiteRockCustom_N}.RGB
   Tangent
   WorldPositionOffset=Multiply(A=Normalize(VectorInput=@n25), B=ScalarParameter() {defaultValue=4, parameterName=ZFightOffset})
   SubsurfaceColor
   ClearCoat
   ClearCoatRoughness
   AmbientOcclusion
   Refraction
   CustomizedUVs_0
   CustomizedUVs_1
   CustomizedUVs_2
   CustomizedUVs_3
   CustomizedUVs_4
   CustomizedUVs_5
   CustomizedUVs_6
   CustomizedUVs_7
   PixelDepthOffset
   ShadingModel
   Displacement
  )
  MaterialFunctionCall(
   Base Material=MaterialFunctionCall(
    Base Material=MaterialFunctionCall!(In=@n12.Blended Material, Height=@n4.Height, UVs=@n3, TransitionGradient=@n13.R) {materialFunction=/Game/Environments/Sets/Vulcano/Materials/MF_LavaEdgeGradient}.Result
    Top Material=@n26
    Alpha=@n24.Alpha
   ) {materialFunction=&MatLayerBlend_Standard}.Blended Material
   Top Material=@n26
   Alpha=@n24.Alpha
  ) {materialFunction=&MatLayerBlend_Standard}
  MaterialFunctionCall(Transition Phase=@n17, Height Texture=@n15.A, Contrast=@n23) {materialFunction=&HeightLerp}
  n27 = ObjectPositionWS()
  n28 = Frac(None=Divide(A=ComponentMask(None=@n27) {b=true}) {constB=384})
  Min(A=OneMinus(None=@n28), B=@n28)
  Multiply(B=@n16)
  n29 = Subtract(A=WorldPosition().XYZ, B=@n27)
  n30 = Multiply(A=@n28) {constB=384}
  n31 = ScalarParameter() {defaultValue=-322.1895, parameterName=EdgeGradientOffset}
  n32 = Divide(
   A=Add(A=ComponentMask(None=@n29) {b=true}, B=Add(A=@n30, B=@n31))
   B=ScalarParameter() {defaultValue=129.9158, parameterName=EdgeGradientLength}
  )
  Multiply(A=OneMinus(None=Clamp(None=@n32)), B=@n13.R)
  OneMinus(None=@n17)
  OneMinus(None=VertexColor().B)
  n33 = GetMaterialAttributes(None=@n22.Result) {attributeGetTypes [#1]}
  SetMaterialAttributes!(MaterialAttributes=@n33.MaterialAttributes, Base Color=Multiply(A=@n33.Base Color, B=@n1.RGB), Opacity Mask=@n24.Alpha) {attributeSetTypes [#1, #2]}
 ]
}
