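# Metal material for the Volcano silo fan that spins the mesh in the vertex shader.
# Time times a Rotation Speed parameter drives RotateAboutAxis into World Position Offset, with Pivot Painter's per-object X-axis vector feeding both axis and pivot point.
# FixRotateAboutAxisNormals rotates the normal map to match.
# The mask texture's green channel alone drives Metallic, Specular and Roughness, and attributes pass through Fortnite_BaseFunction as a gameplay asset.
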Material /Figment_S09/Athena/Environments/Volcano/Materials/M_MI_Volcano_Silo_fan
graph {
 properties {
  bUseMaterialAttributes=true
  physMaterial=/Game/FortressPhysicalMaterials/Metal
  pixelDepthOffsetMode=PDOM_Legacy
  refractionMethod=RM_IndexOfRefraction
  usedShadingModels=Default Lit
 }
 nodes [
  n1 = TextureSampleParameter2D() {parameterName=SpecularMasks, samplerType=SAMPLERTYPE_Masks, texture=/Game/Athena/Environments/Volcano/Textures/Silo/T_volcano_silo_fan_S}
  n2 = MaterialFunctionCall() {materialFunction=/Engine/Functions/Engine_MaterialFunctions02/PivotPainter/PivotPainter_PerObjectData}
  n3 = Multiply(A=Time(), B=ScalarParameter() {defaultValue=1, parameterName=Rotation Speed})
  n4 = TextureSampleParameter2D() {parameterName=Normals, samplerType=SAMPLERTYPE_Normal, texture=/Game/Athena/Environments/Volcano/Textures/Silo/T_volcano_silo_fan_N}
  n5 = Transform(None=@n4.RGB)
  n6 = WorldPosition()
  MaterialFunctionCall!(
   Material attributes=MakeMaterialAttributes!(
    BaseColor=TextureSampleParameter2D() {parameterName=Diffuse, texture=/Game/Athena/Environments/Volcano/Textures/Silo/T_volcano_silo_fan_D}.RGB
    Metallic=@n1.G
    Specular=@n1.G
    Roughness=@n1.G
    Anisotropy
    EmissiveColor
    Opacity
    OpacityMask
    Normal=MaterialFunctionCall(Rotation Axis=@n2.X-Axis Vector, Rotation Angle=@n3, World Space Vertex Normals=@n5) {materialFunction=&FixRotateAboutAxisNormals}.New Tangent Space Vertex Normal
    Tangent
    WorldPositionOffset=RotateAboutAxis(NormalizedRotationAxis=@n2.X-Axis Vector, RotationAngle=@n3, PivotPoint=@n2.X-Axis Vector, Position=@n6.XYZ)
    SubsurfaceColor
    ClearCoat
    ClearCoatRoughness
    AmbientOcclusion
    Refraction
    CustomizedUVs_0
    CustomizedUVs_1
    CustomizedUVs_2
    CustomizedUVs_3
    CustomizedUVs_4
    CustomizedUVs_5
    CustomizedUVs_6
    CustomizedUVs_7
    PixelDepthOffset
    ShadingModel
    Displacement
   )
   IsGameplayAsset=StaticBool() {value=true}
  ) {materialFunction=/Game/Packages/Fortress_SharedMaterials/Base_Material_Functions/Fortnite_BaseFunction}
 ]
}
